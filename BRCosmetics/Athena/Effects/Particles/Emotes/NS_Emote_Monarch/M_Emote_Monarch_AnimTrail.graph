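# Additive unlit ribbon material for the animated trail in the Monarch emote's Niagara effect.
# A blast texture, sampled with swapped UV axes, pans over time with a per-particle random offset.
# Its mask is multiplied by a cubed falloff built from length fade, soft reflected side fade and view-facing angle.
# Inverting that mask drives a hue shift on particle color before emissive adjustments, while particle alpha multiplies the mask.
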Material /BRCosmetics/Athena/Effects/Particles/Emotes/NS_Emote_Monarch/M_Emote_Monarch_AnimTrail
graph {
 properties {bUsedWithNiagaraRibbons=true, blendMode=BLEND_Additive, pixelDepthOffsetMode=PDOM_Legacy, shadingModel=MSM_Unlit, usedShadingModels=Unlit}
 nodes [
  n1 = TextureCoordinate() {coordinateIndex=1}
  n2 = ComponentMask(None=@n1) {r=true}
  MaterialFunctionCall!(Gradient=@n2, Center Point=Constant() {r=0.1}) {materialFunction=/Game/Effects/Fort_Effects/MaterialFunctions/MF_GradientReflected}
  n3 = ScalarParameter() {defaultValue=0.07, parameterName=Hue Shift}
  n4 = MaterialFunctionCall(Float2=TextureCoordinate() {vTiling=0.3}) {materialFunction=&BreakOutFloat2Components}
  n5 = Add(A=@n4.G, B=ParticleRandom())
  n6 = Add(A=@n4.R, B=Multiply(A=Time(), B=ScalarParameter() {defaultValue=-0.5, parameterName=Panner Speed}))
  n7 = MaterialFunctionCall(Gradient=ComponentMask(None=@n1) {g=true}, Center Point=Constant() {r=0.5}) {materialFunction=&MF_GradientReflected}
  n8 = ScalarParameter() {defaultValue=0.5, parameterName=Side Fade Softness}
  n9 = Multiply(A=OneMinus(None=@n2), B=Saturate(None=Divide(A=@n7.Result, B=@n8)))
  n10 = Abs(None=DotProduct(A=VertexNormalWS(), B=CameraVectorWS()))
  n11 = Multiply(
   A=TextureSample(UVs=AppendVector(A=@n5, B=@n6)) {samplerType=SAMPLERTYPE_LinearGrayscale, texture=/Game/Effects/Fort_Effects/Textures/T_Emote_Monarch_Blast}.R
   B=Power(Base=Multiply(A=@n9, B=@n10)) {constExponent=3}
  )
  n12 = OneMinus(None=@n11)
  n13 = ParticleColor()
  MaterialFunctionCall(Emissive=MaterialFunctionCall(Hue Shift Percentage=Multiply(A=@n3, B=@n12), Texture=@n13.RGB) {materialFunction=&HueShift}.Result) {materialFunction=&MF_FX_EmissiveAdjustments}
  Multiply(A=@n13.A, B=@n11)
  SquareRoot(None=@n7.Result)
 ]
}
