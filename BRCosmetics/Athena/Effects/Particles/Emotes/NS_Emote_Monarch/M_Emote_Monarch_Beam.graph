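# Unlit additive Niagara sprite material for a soft beam in the Monarch emote effect.
# A rounded profile across U multiplies an off-center reflected gradient along V to form the beam shape procedurally, without textures.
# That shape is depth-faded, with fade distance blended between two parameters by the cross profile, then squared and scaled by particle alpha.
# Particle color passes through an emissive adjustment function.
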Material /BRCosmetics/Athena/Effects/Particles/Emotes/NS_Emote_Monarch/M_Emote_Monarch_Beam
graph {
 properties {bUsedWithNiagaraSprites=true, blendMode=BLEND_Additive, pixelDepthOffsetMode=PDOM_Legacy, shadingModel=MSM_Unlit, usedShadingModels=Unlit}
 nodes [
  n1 = TextureCoordinate()
  n2 = Multiply(A=MaterialFunctionCall!(Gradient=ComponentMask(None=@n1) {g=true}, Center Point=Constant() {r=0.8}) {materialFunction=/Game/Effects/Fort_Effects/MaterialFunctions/MF_GradientReflected}.Result)
  DotProduct(A=@n2, B=@n2)
  LinearInterpolate()
  n3 = ParticleColor()
  MaterialFunctionCall(Emissive=@n3.RGB) {materialFunction=&MF_FX_EmissiveAdjustments}
  n4 = ConstantBiasScale(None=ComponentMask(None=@n1) {r=true}) {bias=-0.5, scale=2}
  n5 = SquareRoot(None=OneMinus(None=DotProduct(A=@n4, B=@n4)))
  n6 = ScalarParameter() {defaultValue=33, parameterName=Depth Fade Distance B}
  n7 = DepthFade(
   Opacity=Multiply(A=@n2, B=Multiply(A=@n5, B=@n5))
   FadeDistance=LinearInterpolate(A=ScalarParameter() {defaultValue=99, parameterName=Depth Fade Distance A}, B=@n6, Alpha=@n5)
  )
  Multiply(A=@n3.A, B=Multiply(A=@n7, B=@n7))
  Multiply(A=OneMinus(None=@n5), B=@n6)
  SquareRoot(None=@n2)
 ]
}
