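# Unlit translucent Niagara sprite material for spark particles in the Monarch emote effect.
# Particle color RGB feeds the shared MF_FX_EmissiveAdjustments function.
# A 4x2 dust-mote texture supplies alpha, which one branch multiplies by particle alpha.
# A separate branch squares that texture alpha twice to a fourth power; the graph shows no links from any branch to material outputs.
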Material /BRCosmetics/Athena/Effects/Particles/Emotes/NS_Emote_Monarch/M_Emote_Monarch_Spark
graph {
 properties {bUsedWithNiagaraSprites=true, blendMode=BLEND_Translucent, pixelDepthOffsetMode=PDOM_Legacy, shadingModel=MSM_Unlit, usedShadingModels=Unlit}
 nodes [
  n1 = ParticleColor()
  MaterialFunctionCall(Emissive=@n1.RGB) {materialFunction=&MF_FX_EmissiveAdjustments}
  n2 = TextureSample() {texture=/Game/Effects/Fort_Effects/Textures/T_DustMotes_4x2_01}
  Multiply(A=@n1.A, B=@n2.A)
  n3 = Multiply(A=@n2.A, B=@n2.A)
  Multiply(A=@n3, B=@n3)
 ]
}
