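# Niagara system for the warm orange energy blast effect of the Monarch emote.
# Three enabled emitters each fire a delayed one-shot burst: a lens flare sprite, an expanding sphere sprite and a light renderer flash.
# Curves over particle life grow the flare and sphere while fading all three through alpha.
# Blast, Blast_Spokes and Energy_Bits emitters remain in the system but are disabled.
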NiagaraSystem /BRCosmetics/Athena/Effects/Particles/Emotes/NS_Emote_Monarch/NS_Emote_Monarch_Blast
dependencies {
 usedDataInterfaces [&NiagaraDataInterfaceCurve, &NiagaraDataInterfaceVector2DCurve, &NiagaraDataInterfaceColorCurve]
 usedDynamicInputs []
 usedModules [&SystemState]
}
emitters {
 Blast {
  bEnabled=false
  emitterUpdateScript [
   &EmitterState()
   &SpawnBurst_Instantaneous(Use Spawn Probability=false, UseLoopCountLimit=false)
  ]
  particleSpawnScript [&InitializeParticle(
    Color=<rgba 6 2.314 0.18 1>
    Lifetime=3
    Position Offset=<0 7 0>
    Position Offset Coordinate Space=::Local
    Sprite Size Mode=::NewEnumerator3
    Uniform Sprite Size=77
    UsePositionOffset=true
    Write Material Random=true
   )]
  particleUpdateScript [
   &ParticleState()
   &ScaleColor(
    Scale Alpha=&FloatFromCurve(FloatCurve=DataInterface() {propertyValues {Curve {keys [{arriveTangent=-6.073, interpMode=RCIM_Cubic, leaveTangent=-6.073, tangentMode=RCTM_User, time=0, value=4}, <at 1 1>]}}})
    ScaleA=true
    ScaleRGB=true
   )
   &CameraOffset(Camera Offset Amount=11)
  ]
  rendererClasses [&NiagaraSpriteRendererProperties]
  renderers [{rendererClass=&NiagaraSpriteRendererProperties}]
 }
 Blast_Spokes {
  bEnabled=false
  emitterUpdateScript [
   &EmitterState()
   &SpawnBurst_Instantaneous(Spawn Count=5, Spawn Time=0.1, Use Spawn Probability=false, UseLoopCountLimit=false)
  ]
  particleSpawnScript [&InitializeParticle(
    Color=<rgba 6 2.314 0.18 1>
    Lifetime=0.5
    Position Offset=<0 7 0>
    Position Offset Coordinate Space=::Local
    Sprite Size=<44 66>
    Sprite Size Mode=::NewEnumerator4
    UsePositionOffset=true
    Write Material Random=true
   )]
  particleUpdateScript [
   &ParticleState()
   &ScaleColor(
    Scale Alpha=&FloatFromCurve(FloatCurve=DataInterface() {propertyValues {Curve {keys [{arriveTangent=0.1467, interpMode=RCIM_Cubic, leaveTangent=0.1467, tangentMode=RCTM_User, time=0, value=2}, <at 1 0>]}}})
    ScaleA=true
    ScaleRGB=true
   )
   &CameraOffset(Camera Offset Amount=11)
   /SetVariables(Particles.SpriteRotation:NiagaraFloat=&Multiply_Float(A=/Game/Effects/Niagara/DynamicInputs/Math/FN_NormalizedParticleCount(), B=360)) {isSetParameters=true}
   &ScaleSpriteSize()
  ]
  rendererClasses [&NiagaraSpriteRendererProperties]
  renderers [{rendererClass=&NiagaraSpriteRendererProperties}]
 }
 Energy_Bits {
  bEnabled=false
  emitterUpdateScript [
   &EmitterState()
   &SpawnBurst_Instantaneous(Spawn Count=99, Spawn Time=0.266, Use Spawn Probability=false, UseLoopCountLimit=false)
  ]
  particleSpawnScript [
   &InitializeParticle(
    Color=<rgba 11 4.521 0.77 1>
    Lifetime Max=1.1
    Lifetime Min=0.033
    Lifetime Mode=::NewEnumerator1
    Sprite Size Max=<2 111>
    Sprite Size Min=<1 55>
    Sprite Size Mode=::NewEnumerator2
    UsePositionOffset=false
    Write Material Random=true
   )
   &AddVelocityFromPoint(Module.Use Falloff Distance=false, Velocity Strength=&RandomRangeFloat(Minimum=750, Maximum=1000))
   &SphereLocation(
    MaskSpawnGroup=false
    Override Seed=false
    Sphere Radius=1
    Surface Only=false
    UseNonUniformScale=false
    UseOffset=false
    UseSphereOrientation=false
   )
   &SubUVAnimation(SubUV Animation Mode:ENiagara_SubUVLookupMode=::NewEnumerator2)
  ]
  particleUpdateScript [
   &Drag(Drag=&RandomRangeFloat(Minimum=2, Maximum=7), Use Linear Drag=true, Use Rotational Drag=false)
   &SolveForcesAndVelocity(Clamp Velocity=false, Limit Acceleration=false)
   &ParticleState()
   &ScaleSpriteSize(
    Non-Uniform Curve Sprite Scale=DataInterface() {propertyValues {
      XCurve {keys [
        <at 0 0>
        {interpMode=RCIM_Cubic, tangentMode=RCTM_User, time=0.1, value=1}
        {arriveTangent=-3.2987, interpMode=RCIM_Cubic, leaveTangent=-3.2987, tangentMode=RCTM_User, time=1, value=0}
       ]}
      YCurve {keys [{arriveTangent=-3.6796, interpMode=RCIM_Cubic, leaveTangent=-3.6796, tangentMode=RCTM_User, time=0, value=1}, <at 1 0>]}
     }}
    Scale Sprite Size Mode=::NewEnumerator3
   )
   &ScaleColor(
    Scale Alpha=&Multiply_Float(
     A=&RandomRangeFloat(Minimum=0.5, Maximum=3)
     B=&FloatFromCurve(FloatCurve=DataInterface() {propertyValues {Curve {keys [<at 0.1 0>, {arriveTangent=-3.9594, interpMode=RCIM_Cubic, leaveTangent=-3.9594, tangentMode=RCTM_User, time=0.2, value=1}, <at 1 0>]}}})
    )
    ScaleA=true
    ScaleRGB=false
   )
  ]
  rendererClasses [&NiagaraSpriteRendererProperties]
  renderers [{rendererClass=&NiagaraSpriteRendererProperties}]
 }
 Lens_Flare {
  emitterUpdateScript [
   &EmitterState()
   &SpawnBurst_Instantaneous(Spawn Time=0.166, Use Spawn Probability=false, UseLoopCountLimit=false)
  ]
  particleSpawnScript [&InitializeParticle(
    Color=<rgba 9 5.5575 1.35 1>
    Position Offset=<0 7 0>
    Position Offset Coordinate Space=::Local
    Sprite Size Mode=::NewEnumerator3
    Uniform Sprite Size=222
    UsePositionOffset=true
    Write Material Random=true
   )]
  particleUpdateScript [
   &ParticleState()
   &ScaleColor(
    Scale Alpha=&FloatFromCurve(FloatCurve=DataInterface() {propertyValues {Curve {keys [{arriveTangent=-3.0287, interpMode=RCIM_Cubic, leaveTangent=-3.0287, tangentMode=RCTM_User, time=0, value=1}, <at 1 0>]}}})
    ScaleA=true
    ScaleRGB=false
   )
   &CameraOffset(Camera Offset Amount=22)
   &ScaleSpriteSize(Uniform Curve Sprite Scale=DataInterface() {propertyValues {Curve {keys [
        {arriveTangent=3.5616, interpMode=RCIM_Cubic, leaveTangent=3.5616, tangentMode=RCTM_User, time=0, value=0}
        {tangentMode=RCTM_User, time=1, value=1}
       ]}}})
  ]
  rendererClasses [&NiagaraSpriteRendererProperties]
  renderers [{rendererClass=&NiagaraSpriteRendererProperties}]
 }
 Light {
  emitterUpdateScript [
   &EmitterState()
   &SpawnBurst_Instantaneous(Spawn Time=0.266, Use Spawn Probability=false, UseLoopCountLimit=false)
  ]
  particleSpawnScript [&InitializeParticle(
    Color=<rgba 11 4.235 1.1 1>
    Lifetime=1.133
    Position Offset=<0 22 0>
    Sprite Size Mode=::NewEnumerator3
    Uniform Sprite Size=222
    UsePositionOffset=true
    Write Material Random=true
   )]
  particleUpdateScript [
   &ParticleState()
   &ScaleColor(
    Scale Alpha=&FloatFromCurve(FloatCurve=DataInterface() {propertyValues {Curve {keys [{arriveTangent=-2.4402, interpMode=RCIM_Cubic, leaveTangent=-2.4402, tangentMode=RCTM_User, time=0, value=4}, <at 0.1 0.75>, <at 1 0>]}}})
    ScaleA=true
    ScaleRGB=false
   )
  ]
  rendererClasses [&NiagaraLightRendererProperties]
  renderers [{rendererClass=&NiagaraLightRendererProperties, rendererIndex=0}]
 }
 Sphere {
  emitterUpdateScript [
   &EmitterState()
   &SpawnBurst_Instantaneous(Spawn Time=0.233, Use Spawn Probability=false, UseLoopCountLimit=false)
  ]
  particleSpawnScript [&InitializeParticle(
    Color=<rgba 0.6 0.21 0 0.8>
    Lifetime=1.3
    Sprite Size Mode=::NewEnumerator3
    Uniform Sprite Size=222
    UsePositionOffset=false
    Write Material Random=true
   )]
  particleUpdateScript [
   &ParticleState()
   &ScaleSpriteSize(Uniform Curve Sprite Scale=DataInterface() {propertyValues {Curve {keys [
        {arriveTangent=4.1785, interpMode=RCIM_Cubic, leaveTangent=4.1785, tangentMode=RCTM_User, time=0, value=0}
        {arriveTangent=0.9995, interpMode=RCIM_Cubic, leaveTangent=0.9995, tangentMode=RCTM_User, time=0.2223, value=0.6825}
        {arriveTangent=-0.0405, interpMode=RCIM_Cubic, leaveTangent=-0.0405, tangentMode=RCTM_User, time=1, value=1}
       ]}}})
   &ScaleColor(
    Scale Alpha=&FloatFromCurve(FloatCurve=DataInterface() {propertyValues {Curve {keys [
         {arriveTangent=-2.9427, interpMode=RCIM_Cubic, leaveTangent=-2.9427, leaveTangentWeight=0.9539, tangentMode=RCTM_User, time=0, value=1}
         {arriveTangentWeight=1.1112, interpMode=RCIM_Cubic, tangentMode=RCTM_User, time=1, value=0}
        ]}}})
    ScaleA=true
    ScaleRGB=true
   )
  ]
  rendererClasses [&NiagaraSpriteRendererProperties]
  renderers [{rendererClass=&NiagaraSpriteRendererProperties}]
 }
}
summary {
 systemName=NS_Emote_Monarch_Blast
 emitters [
  {emitterName=Sphere, rendererClasses [&NiagaraSpriteRendererProperties]}
  {bEnabled=false, emitterName=Blast, rendererClasses [&NiagaraSpriteRendererProperties]}
  {emitterName=Light, rendererClasses [&NiagaraLightRendererProperties]}
  {emitterName=Lens_Flare, rendererClasses [&NiagaraSpriteRendererProperties]}
  {bEnabled=false, emitterName=Energy_Bits, rendererClasses [&NiagaraSpriteRendererProperties]}
  {bEnabled=false, emitterName=Blast_Spokes, rendererClasses [&NiagaraSpriteRendererProperties]}
 ]
 userVariables []
}
