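# Niagara effect for the Monarch emote: golden bracelet trails and glows, plus a burst of light and energy.
# Two anim-trail ribbons use bone pairs with a loop-age alpha fade, while two glow sprites track those bones on User.SkeletalMesh with mirrored offsets.
# A vertically aligned beam, a soft light sprite and upward-sprayed, drag-slowed bits from a cylinder complete the burst.
# Curves drive the fades and sprite scaling.
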NiagaraSystem /BRCosmetics/Athena/Effects/Particles/Emotes/NS_Emote_Monarch/NS_Emote_Monarch_Bracelet
dependencies {
 usedDataInterfaces [&NiagaraDataInterfaceCurve, &NiagaraDataInterfaceVector2DCurve, &NiagaraDataInterfaceColorCurve]
 usedDynamicInputs []
 usedModules [&SystemState]
}
emitters {
 Beam {
  emitterUpdateScript [
   &EmitterState()
   &SpawnBurst_Instantaneous(Spawn Time=0.13, Use Spawn Probability=false, UseLoopCountLimit=false)
  ]
  particleSpawnScript [&InitializeParticle(
    Color=<rgba 1 0.84 0.4 0.1>
    Lifetime=2
    Sprite Size=<111 333>
    Sprite Size Mode=::NewEnumerator4
    UsePositionOffset=false
    Write Material Random=true
   )]
  particleUpdateScript [
   &ParticleState()
   &SpriteFacingAndAlignment(Alignment Coordinate Space=::World, Sprite Alignment=<0 0 1>, Sprite Facing=<0 0 0>, WriteAlignment=true, WriteFacing=false)
   &CameraOffset(Camera Offset Amount=99)
   &ScaleSpriteSize(
    Non-Uniform Curve Sprite Scale=DataInterface() {propertyValues {
      XCurve {keys [{tangentMode=RCTM_User, time=0, value=1}, {tangentMode=RCTM_User, time=1, value=1}]}
      YCurve {keys [
        {arriveTangent=1.9423, interpMode=RCIM_Cubic, leaveTangent=1.9423, tangentMode=RCTM_User, time=0, value=0.5}
        {tangentMode=RCTM_User, time=0.7, value=1}
       ]}
     }}
    Scale Sprite Size Mode=::NewEnumerator3
   )
   &ScaleColor(
    Scale Alpha=&FloatFromCurve(FloatCurve=DataInterface() {propertyValues {Curve {keys [
         {arriveTangent=10.3113, interpMode=RCIM_Cubic, leaveTangent=10.3113, tangentMode=RCTM_User, time=0, value=0}
         {arriveTangent=-0.0454, interpMode=RCIM_Cubic, leaveTangent=-0.0454, tangentMode=RCTM_User, time=0.2, value=1}
         {arriveTangent=-2.3405, interpMode=RCIM_Cubic, leaveTangent=-2.3405, tangentMode=RCTM_User, time=1, value=0}
        ]}}})
    ScaleA=true
    ScaleRGB=false
   )
  ]
  rendererClasses [&NiagaraSpriteRendererProperties]
  renderers [{rendererClass=&NiagaraSpriteRendererProperties}]
 }
 Beam_Light {
  emitterUpdateScript [
   &EmitterState()
   &SpawnBurst_Instantaneous(Use Spawn Probability=false, UseLoopCountLimit=false)
  ]
  particleSpawnScript [
   &InitializeParticle(
    Color=<rgba 5 4.6 3.5 1>
    Lifetime=2
    Position Offset=<0 0 10>
    Sprite Size Mode=::NewEnumerator3
    Uniform Sprite Size=450
    UsePositionOffset=true
    Write Material Random=true
   )
   &FN_SetSpawnGroup(Number of Spawn Groups=2)
  ]
  particleUpdateScript [
   &ParticleState(Kill Particles When Lifetime Has Elapsed=false, Let Infinitely Lived Particles Die When Emitter Deactivates=true)
   &ScaleColor(
    Scale Alpha=&FloatFromCurve(FloatCurve=DataInterface() {propertyValues {Curve {keys [
         {arriveTangent=10.3113, interpMode=RCIM_Cubic, leaveTangent=10.3113, tangentMode=RCTM_User, time=0.1, value=0}
         {arriveTangent=-0.0454, interpMode=RCIM_Cubic, leaveTangent=-0.0454, tangentMode=RCTM_User, time=0.3, value=1}
         {arriveTangent=-2.3405, interpMode=RCIM_Cubic, leaveTangent=-2.3405, tangentMode=RCTM_User, time=1, value=0}
        ]}}})
    ScaleA=true
    ScaleRGB=false
   )
   &ScaleSpriteSize(
    Non-Uniform Curve Sprite Scale=DataInterface() {propertyValues {
      XCurve {keys [
        {arriveTangent=1.8486, interpMode=RCIM_Cubic, leaveTangent=1.8486, tangentMode=RCTM_User, time=0, value=0.5}
        <at 0.7 1>
        {tangentMode=RCTM_User, time=1, value=1}
       ]}
      YCurve {keys [
        {arriveTangent=1.9423, interpMode=RCIM_Cubic, leaveTangent=1.9423, tangentMode=RCTM_User, time=0, value=0.5}
        {tangentMode=RCTM_User, time=0.7, value=1}
       ]}
     }}
    Scale Sprite Size Mode=::NewEnumerator3
   )
  ]
  rendererClasses [&NiagaraSpriteRendererProperties]
  renderers [{rendererClass=&NiagaraSpriteRendererProperties}]
 }
 Bracelet_AnimTrail {
  emitterUpdateScript [
   &EmitterState(Life Cycle Mode=::NewEnumerator1, Loop Behavior=::NewEnumerator1, Loop Duration=2, UseLoopDelay=false)
   &FN_SpawnAnimTrail_v2(Number of Anim Trails=2, Spawn Interval=0, Spawn Mode:ENiagaraSkeletalSpawning=::NewEnumerator0)
  ]
  particleSpawnScript [
   /Game/Effects/Niagara/Modules/AnimTrail/FN_InitializeAnimTrail/FN_InitializeAnimTrailBoneIndecies(Second Bone or Socket Index=1)
   &FN_InitializeAnimTrailBoneIndecies(AnimTrail Selection=1, First Bone or Socket Index=2, Second Bone or Socket Index=3) {module=FN_InitializeAnimTrailBoneIndecies001}
   &FN_InitializeAnimTrail(
    AnimTrailWidth=1.1
    Color=&MakeLinearColor(
     R:NiagaraFloat=2
     G:NiagaraFloat=0.75
     B:NiagaraFloat=0
     A:NiagaraFloat=&FloatFromCurve(
      FloatCurve=DataInterface() {propertyValues {Curve {keys [
          {interpMode=RCIM_Cubic, tangentMode=RCTM_User, time=0, value=0}
          {interpMode=RCIM_Cubic, leaveTangent=-3.1837, tangentMode=RCTM_Break, time=0.1, value=1}
          {interpMode=RCIM_Cubic, tangentMode=RCTM_User, time=1, value=0}
         ]}}}
      CurveIndex=$Emitter.NormalizedLoopAge
     )
    )
    Lifetime=3
   )
  ]
  particleUpdateScript [
   /Game/Effects/Niagara/Modules/AnimTrail/FN_UpdateAnimTrail/FN_UpdateAnimTrail_v4/FN_UpdateAnimTrail()
   &ParticleState()
   &ScaleColor(
    Scale Alpha=&FloatFromCurve(FloatCurve=DataInterface() {propertyValues {Curve {keys [
         {interpMode=RCIM_Cubic, tangentMode=RCTM_User, time=0, value=1}
         {arriveTangent=-2.7485, interpMode=RCIM_Cubic, leaveTangent=-2.7485, tangentMode=RCTM_User, time=1, value=0}
        ]}}})
    ScaleA=true
    ScaleRGB=true
   )
  ]
  rendererClasses [&NiagaraRibbonRendererProperties]
  renderers [{rendererClass=&NiagaraRibbonRendererProperties}]
 }
 Bracelet_Lights {
  emitterUpdateScript [
   &EmitterState()
   &SpawnBurst_Instantaneous(Spawn Count=2, Use Spawn Probability=false, UseLoopCountLimit=false)
  ]
  particleSpawnScript [
   &InitializeParticle(
    Color=<rgba 33 21.45 13.2 1>
    Lifetime=2.1
    Sprite Size Mode=::NewEnumerator3
    Uniform Sprite Size=111
    UsePositionOffset=false
    Write Material Random=true
   )
   &FN_SetSpawnGroup(Number of Spawn Groups=2)
  ]
  particleUpdateScript [
   &ParticleState(Kill Particles When Lifetime Has Elapsed=false, Let Infinitely Lived Particles Die When Emitter Deactivates=true)
   &SkeletalMeshLocation(
    Bone / Socket Index=&Multiply_Int(A=&SpawnGroup(), B=2)
    Bone Sampling Mode=::NewEnumerator4
    MaskSpawnGroup=false
    Sampled Position Offset=&MakeVector(
     X=&Multiply_Float(
      A=/Niagara/DynamicInputs/Multiply/ScaleAndBiasFloat(
       Float:NiagaraFloat=&MakeFloatFromInt(Int=&SpawnGroup())
       Scale:NiagaraFloat=2
       Bias:NiagaraFloat=-1
      )
      B=-15
     )
     Y=0
     Z=0
    )
    Skeletal Mesh=$User.SkeletalMesh
   )
   &ScaleColor(
    Scale Alpha=&FloatFromCurve(FloatCurve=DataInterface() {propertyValues {Curve {keys [
         <at 0 0 RCIM_Cubic>
         <at 0.8 1 RCIM_Cubic>
         {arriveTangent=-16.3408, interpMode=RCIM_Cubic, leaveTangent=-16.3408, tangentMode=RCTM_User, time=1, value=0}
        ]}}})
    Scale RGB=&VectorFromFloat(Value=1)
    ScaleA=true
    ScaleRGB=true
   )
  ]
  rendererClasses [&NiagaraSpriteRendererProperties]
  renderers [{rendererClass=&NiagaraSpriteRendererProperties}]
 }
 Energy_Bits {
  emitterUpdateScript [
   &EmitterState()
   &SpawnBurst_Instantaneous(Spawn Count=33, Use Spawn Probability=false, UseLoopCountLimit=false)
  ]
  particleSpawnScript [
   &InitializeParticle(
    Color=<rgba 1 0.559 0.02 1>
    Lifetime Min=0.5
    Lifetime Mode=::NewEnumerator1
    Sprite Size Mode=::NewEnumerator1
    Uniform Sprite Size Max=3
    Uniform Sprite Size Min=1.5
    UsePositionOffset=false
    Write Material Random=true
   )
   &SubUVAnimation(SubUV Animation Mode:ENiagara_SubUVLookupMode=::NewEnumerator2)
   &CylinderLocation(Cylinder Height=44, Cylinder Height Midpoint=0, Cylinder Radius=22, MaskSpawnGroup=false, Override Seed=false, UseOwnerScale=true)
   &AddVelocityInCone(
    Cone Angle=33
    Cone Axis=<0 0 1>
    Cone Axis Coordinate Space=::World
    Module.Use Velocity Falloff On Cone Axis=false
    Velocity Strength=&RandomRangeFloat(Minimum=33, Maximum=333)
   )
  ]
  particleUpdateScript [
   &Drag(Drag=&RandomRangeFloat(Minimum=2, Maximum=4), Use Linear Drag=true, Use Rotational Drag=false)
   &SolveForcesAndVelocity(Clamp Velocity=false, Limit Acceleration=false)
   &ParticleState()
   &ScaleColor(
    Scale Alpha=&Multiply_Float(
     A=&RandomRangeFloat(Minimum=0.2, Maximum=1)
     B=&FloatFromCurve(FloatCurve=DataInterface() {propertyValues {Curve {keys [{arriveTangent=-1.645, interpMode=RCIM_Cubic, leaveTangent=-1.645, tangentMode=RCTM_User, time=0, value=1}, <at 1 0>]}}})
    )
    ScaleA=true
    ScaleRGB=false
   )
   &ScaleSpriteSize(
    Non-Uniform Curve Sprite Scale=DataInterface() {propertyValues {
      XCurve {keys [
        {arriveTangent=1.3399, interpMode=RCIM_Cubic, leaveTangent=1.3399, tangentMode=RCTM_User, time=0, value=0}
        {tangentMode=RCTM_User, time=1, value=1}
       ]}
      YCurve {keys [
        {interpMode=RCIM_Cubic, tangentMode=RCTM_User, time=0, value=2}
        {arriveTangent=-2.0948, interpMode=RCIM_Cubic, leaveTangent=-2.0948, tangentMode=RCTM_User, time=1, value=1}
       ]}
     }}
    Scale Sprite Size Mode=::NewEnumerator3
   )
   &CameraOffset(Camera Offset Amount=44)
  ]
  rendererClasses [&NiagaraSpriteRendererProperties]
  renderers [{rendererClass=&NiagaraSpriteRendererProperties}]
 }
}
summary {
 systemName=NS_Emote_Monarch_Bracelet
 emitters [
  {emitterName=Bracelet_AnimTrail, rendererClasses [&NiagaraRibbonRendererProperties]}
  {emitterName=Bracelet_Lights, rendererClasses [&NiagaraSpriteRendererProperties]}
  {emitterName=Beam, rendererClasses [&NiagaraSpriteRendererProperties]}
  {emitterName=Energy_Bits, rendererClasses [&NiagaraSpriteRendererProperties]}
  {emitterName=Beam_Light, rendererClasses [&NiagaraSpriteRendererProperties]}
 ]
 userVariables [
  User.RibbonIDOffset {type=NiagaraInt32}
  User.SkeletalMesh {
   type=NiagaraDataInterfaceSkeletalMesh
   defaultValue {
    struct=NiagaraExt_VariableValue_DataInterface
    value {dataInterface=&self:NiagaraDataInterfaceSkeletalMesh_0}
   }
  }
 ]
}
